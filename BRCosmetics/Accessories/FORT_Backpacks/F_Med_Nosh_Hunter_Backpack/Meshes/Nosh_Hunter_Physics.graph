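# Physics asset for the Nosh Hunter backpack cosmetic, giving collision bodies and joints to its dangling goo, shells, pouch and handle bones.
# The pack body, a capsule plus box, anchors two tighter-limited shell boxes, the left goo capsule and a rotation-locked handle sphere.
# Right goo hangs from the pouch and branches into three capsules, two swinging and twisting freely.
# No constraint links the pouch to the pack.
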PhysicsAsset /BRCosmetics/Accessories/FORT_Backpacks/F_Med_Nosh_Hunter_Backpack/Meshes/Nosh_Hunter_Physics
bodies {
 dyn_goo_l_01 [{shapeType=Capsule, length=3.6651, radius=2.315, center=<-0.2092 0.1255 -4.3412>}]
 dyn_goo_r_03 [{shapeType=Capsule, length=3.6651, radius=2.315, center=<-0.2092 0.1255 -4.3412>}]
 dyn_goo_r_03_a [{shapeType=Capsule, length=1.133, radius=0.505, center=<-0.0776 0.0459 -0.7983>}]
 dyn_goo_r_03_b_02 [{shapeType=Capsule, length=1.8698, radius=0.7158, center=<0 0 -1.5183>}]
 dyn_goo_r_03_c [{shapeType=Capsule, length=0.8915, radius=0.505, center=<-0.0776 0.0459 -0.9685>}]
 dyn_handle [{shapeType=Sphere, radius=2.0073, center=<0 0.0391 -0.4208>}]
 dyn_pouch_r [{shapeType=Box, extentX=1.1924, extentY=1.2322, extentZ=1.5483, center=<0 0 -0.0491>}]
 dyn_shells_l [{shapeType=Box, extentX=8.6602, extentY=5.2315, extentZ=2, center=<1.1282 0.2743 0.033>, rotation=<rot 90 0 0>}]
 dyn_shells_r [{shapeType=Box, extentX=8.6602, extentY=5.2315, extentZ=2, center=<-0.9753 0.2743 0.033>, rotation=<rot 90 0 0>}]
 pack [{shapeType=Capsule, length=18.7341, radius=9.9269, center=<-0.6084 -1.7214 -8.8743>}, {shapeType=Box, extentX=7.5, extentY=7.5, extentZ=7.5}]
}
constraints [
 {bone1Name=dyn_goo_r_03, bone2Name=dyn_pouch_r, swing1LimitDegrees=45, swing2LimitDegrees=45, twistLimitDegrees=45}
 {
  bone1Name=dyn_goo_r_03_a
  bone2Name=dyn_goo_r_03
  swing1LimitDegrees=45
  swing1Motion=Free
  swing2LimitDegrees=45
  swing2Motion=Free
  twistLimitDegrees=45
  twistMotion=Free
 }
 {bone1Name=dyn_goo_r_03_b_02, bone2Name=dyn_goo_r_03, swing1LimitDegrees=45, swing2LimitDegrees=45, twistLimitDegrees=45}
 {
  bone1Name=dyn_goo_r_03_c
  bone2Name=dyn_goo_r_03
  swing1LimitDegrees=45
  swing1Motion=Free
  swing2LimitDegrees=45
  swing2Motion=Free
  twistLimitDegrees=45
  twistMotion=Free
 }
 {bone1Name=dyn_shells_l, bone2Name=pack, swing1LimitDegrees=25, swing2LimitDegrees=10, twistLimitDegrees=18}
 {bone1Name=dyn_shells_r, bone2Name=pack, swing1LimitDegrees=25, swing2LimitDegrees=10, twistLimitDegrees=18}
 {bone1Name=dyn_goo_l_01, bone2Name=pack, swing1LimitDegrees=45, swing2LimitDegrees=45, twistLimitDegrees=45}
 {
  bone1Name=dyn_handle
  bone2Name=pack
  swing1LimitDegrees=45
  swing1Motion=Locked
  swing2LimitDegrees=45
  swing2Motion=Locked
  twistLimitDegrees=45
  twistMotion=Locked
 }
]
